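# Physics asset for the evil-upgrade astronaut backpack cosmetic, defining physics bodies for its seven tentacles (three left, three right, one back).
# Each tentacle is a chain of five capsules, mostly tapering toward the tip and joined by angular swing and twist limits, rooted on one larger pack capsule.
# Six root joints leave swing and twist free, while left tentacle A's root is the exception with angular limits and tighter twist.
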PhysicsAsset /BRCosmetics/Accessories/FORT_Backpacks/F_MED_AstronautEvilUpgrade_Backpack/Meshes/F_MED_Astronaut_Evil_Upgrade_Pack_Physics
bodies {
 dyn_bk_tentacle_01 [{shapeType=Capsule, length=1.4603, radius=2.5397, center=<2 0 0>, rotation=<rot 90 0 0>}]
 dyn_bk_tentacle_02 [{shapeType=Capsule, length=2, radius=2, center=<4 0 0>, rotation=<rot 90 0 0>}]
 dyn_bk_tentacle_03 [{shapeType=Capsule, length=2.3195, radius=1.6805, center=<2 0 0>, rotation=<rot 90 0 0>}]
 dyn_bk_tentacle_04 [{shapeType=Capsule, length=2.5226, radius=1.4774, center=<2 0 0>, rotation=<rot 90 0 0>}]
 dyn_bk_tentacle_05 [{shapeType=Capsule, length=2.8494, radius=1.1506, center=<2 0 0>, rotation=<rot 90 0 0>}]
 dyn_l_tentacle_a_01 [{shapeType=Capsule, length=1.4603, radius=2.5397, center=<2 0 0>, rotation=<rot 90 0 0>}]
 dyn_l_tentacle_a_02 [{shapeType=Capsule, length=2, radius=2, center=<4 0 0>, rotation=<rot 90 0 0>}]
 dyn_l_tentacle_a_03 [{shapeType=Capsule, length=2.3195, radius=1.6805, center=<2 0 0>, rotation=<rot 90 0 0>}]
 dyn_l_tentacle_a_04 [{shapeType=Capsule, length=2.5226, radius=1.4774, center=<2 0 0>, rotation=<rot 90 0 0>}]
 dyn_l_tentacle_a_05 [{shapeType=Capsule, length=2.8494, radius=1.1506, center=<2 0 0>, rotation=<rot 90 0 0>}]
 dyn_l_tentacle_b_01 [{shapeType=Capsule, length=1.4603, radius=2.5397, center=<2 0 0>, rotation=<rot 90 0 0>}]
 dyn_l_tentacle_b_02 [{shapeType=Capsule, length=2, radius=2, center=<4 0 0>, rotation=<rot 90 0 0>}]
 dyn_l_tentacle_b_03 [{shapeType=Capsule, length=2.3195, radius=1.6805, center=<2 0 0>, rotation=<rot 90 0 0>}]
 dyn_l_tentacle_b_04 [{shapeType=Capsule, length=2.5226, radius=1.4774, center=<2 0 0>, rotation=<rot 90 0 0>}]
 dyn_l_tentacle_b_05 [{shapeType=Capsule, length=2.8494, radius=1.1506, center=<2 0 0>, rotation=<rot 90 0 0>}]
 dyn_l_tentacle_c_01 [{shapeType=Capsule, length=1.4603, radius=2.5397, center=<2 0 0>, rotation=<rot 90 0 0>}]
 dyn_l_tentacle_c_02 [{shapeType=Capsule, length=2, radius=2, center=<4 0 0>, rotation=<rot 90 0 0>}]
 dyn_l_tentacle_c_03 [{shapeType=Capsule, length=2.3195, radius=1.6805, center=<2 0 0>, rotation=<rot 90 0 0>}]
 dyn_l_tentacle_c_04 [{shapeType=Capsule, length=2.5226, radius=1.4774, center=<2 0 0>, rotation=<rot 90 0 0>}]
 dyn_l_tentacle_c_05 [{shapeType=Capsule, length=2.8494, radius=1.1506, center=<2 0 0>, rotation=<rot 90 0 0>}]
 dyn_r_tentacle_a_01 [{shapeType=Capsule, length=1.4603, radius=2.5397, center=<2 0 0>, rotation=<rot 90 0 0>}]
 dyn_r_tentacle_a_02 [{shapeType=Capsule, length=2, radius=2, center=<4 0 0>, rotation=<rot 90 0 0>}]
 dyn_r_tentacle_a_03 [{shapeType=Capsule, length=2.3195, radius=1.6805, center=<2 0 0>, rotation=<rot 90 0 0>}]
 dyn_r_tentacle_a_04 [{shapeType=Capsule, length=2.5226, radius=1.4774, center=<2 0 0>, rotation=<rot 90 0 0>}]
 dyn_r_tentacle_a_05 [{shapeType=Capsule, length=2.8494, radius=1.1506, center=<2 0 0>, rotation=<rot 90 0 0>}]
 dyn_r_tentacle_b_01 [{shapeType=Capsule, length=1.4603, radius=2.5397, center=<2 0 0>, rotation=<rot 90 0 0>}]
 dyn_r_tentacle_b_02 [{shapeType=Capsule, length=2, radius=2, center=<4 0 0>, rotation=<rot 90 0 0>}]
 dyn_r_tentacle_b_03 [{shapeType=Capsule, length=2.3195, radius=1.6805, center=<2 0 0>, rotation=<rot 90 0 0>}]
 dyn_r_tentacle_b_04 [{shapeType=Capsule, length=2.5226, radius=1.4774, center=<2 0 0>, rotation=<rot 90 0 0>}]
 dyn_r_tentacle_b_05 [{shapeType=Capsule, length=2.8494, radius=1.1506, center=<2 0 0>, rotation=<rot 90 0 0>}]
 dyn_r_tentacle_c_01 [{shapeType=Capsule, length=1.4603, radius=2.5397, center=<2 0 0>, rotation=<rot 90 0 0>}]
 dyn_r_tentacle_c_02 [{shapeType=Capsule, length=2, radius=2, center=<4 0 0>, rotation=<rot 90 0 0>}]
 dyn_r_tentacle_c_03 [{shapeType=Capsule, length=2.3195, radius=1.6805, center=<2 0 0>, rotation=<rot 90 0 0>}]
 dyn_r_tentacle_c_04 [{shapeType=Capsule, length=2.5226, radius=1.4774, center=<2 0 0>, rotation=<rot 90 0 0>}]
 dyn_r_tentacle_c_05 [{shapeType=Capsule, length=2.8494, radius=1.1506, center=<2 0 0>, rotation=<rot 90 0 0>}]
 pack [{shapeType=Capsule, length=5.7305, radius=10.9603, center=<0 -9.6661 1.8971>}]
}
constraints [
 {bone1Name=dyn_l_tentacle_a_01, bone2Name=pack, swing1LimitDegrees=30, swing2LimitDegrees=30, twistLimitDegrees=5}
 {bone1Name=dyn_l_tentacle_a_02, bone2Name=dyn_l_tentacle_a_01, swing1LimitDegrees=30, swing2LimitDegrees=30, twistLimitDegrees=10}
 {bone1Name=dyn_l_tentacle_a_03, bone2Name=dyn_l_tentacle_a_02, swing1LimitDegrees=30, swing2LimitDegrees=30, twistLimitDegrees=10}
 {bone1Name=dyn_l_tentacle_a_04, bone2Name=dyn_l_tentacle_a_03, swing1LimitDegrees=30, swing2LimitDegrees=30, twistLimitDegrees=10}
 {bone1Name=dyn_l_tentacle_a_05, bone2Name=dyn_l_tentacle_a_04, swing1LimitDegrees=30, swing2LimitDegrees=30, twistLimitDegrees=10}
 {
  bone1Name=dyn_l_tentacle_b_01
  bone2Name=pack
  swing1LimitDegrees=45
  swing1Motion=Free
  swing2LimitDegrees=45
  swing2Motion=Free
  twistLimitDegrees=45
  twistMotion=Free
 }
 {bone1Name=dyn_l_tentacle_b_02, bone2Name=dyn_l_tentacle_b_01, swing1LimitDegrees=30, swing2LimitDegrees=30, twistLimitDegrees=10}
 {bone1Name=dyn_l_tentacle_b_03, bone2Name=dyn_l_tentacle_b_02, swing1LimitDegrees=30, swing2LimitDegrees=30, twistLimitDegrees=10}
 {bone1Name=dyn_l_tentacle_b_04, bone2Name=dyn_l_tentacle_b_03, swing1LimitDegrees=30, swing2LimitDegrees=30, twistLimitDegrees=10}
 {bone1Name=dyn_l_tentacle_b_05, bone2Name=dyn_l_tentacle_b_04, swing1LimitDegrees=30, swing2LimitDegrees=30, twistLimitDegrees=10}
 {
  bone1Name=dyn_l_tentacle_c_01
  bone2Name=pack
  swing1LimitDegrees=45
  swing1Motion=Free
  swing2LimitDegrees=45
  swing2Motion=Free
  twistLimitDegrees=45
  twistMotion=Free
 }
 {bone1Name=dyn_l_tentacle_c_02, bone2Name=dyn_l_tentacle_c_01, swing1LimitDegrees=30, swing2LimitDegrees=30, twistLimitDegrees=10}
 {bone1Name=dyn_l_tentacle_c_03, bone2Name=dyn_l_tentacle_c_02, swing1LimitDegrees=30, swing2LimitDegrees=30, twistLimitDegrees=10}
 {bone1Name=dyn_l_tentacle_c_04, bone2Name=dyn_l_tentacle_c_03, swing1LimitDegrees=30, swing2LimitDegrees=30, twistLimitDegrees=10}
 {bone1Name=dyn_l_tentacle_c_05, bone2Name=dyn_l_tentacle_c_04, swing1LimitDegrees=30, swing2LimitDegrees=30, twistLimitDegrees=10}
 {
  bone1Name=dyn_bk_tentacle_01
  bone2Name=pack
  swing1LimitDegrees=45
  swing1Motion=Free
  swing2LimitDegrees=45
  swing2Motion=Free
  twistLimitDegrees=45
  twistMotion=Free
 }
 {bone1Name=dyn_bk_tentacle_02, bone2Name=dyn_bk_tentacle_01, swing1LimitDegrees=30, swing2LimitDegrees=30, twistLimitDegrees=10}
 {bone1Name=dyn_bk_tentacle_03, bone2Name=dyn_bk_tentacle_02, swing1LimitDegrees=30, swing2LimitDegrees=30, twistLimitDegrees=10}
 {bone1Name=dyn_bk_tentacle_04, bone2Name=dyn_bk_tentacle_03, swing1LimitDegrees=30, swing2LimitDegrees=30, twistLimitDegrees=10}
 {bone1Name=dyn_bk_tentacle_05, bone2Name=dyn_bk_tentacle_04, swing1LimitDegrees=30, swing2LimitDegrees=30, twistLimitDegrees=10}
 {
  bone1Name=dyn_r_tentacle_a_01
  bone2Name=pack
  swing1LimitDegrees=45
  swing1Motion=Free
  swing2LimitDegrees=45
  swing2Motion=Free
  twistLimitDegrees=45
  twistMotion=Free
 }
 {bone1Name=dyn_r_tentacle_a_02, bone2Name=dyn_r_tentacle_a_01, swing1LimitDegrees=30, swing2LimitDegrees=30, twistLimitDegrees=10}
 {bone1Name=dyn_r_tentacle_a_03, bone2Name=dyn_r_tentacle_a_02, swing1LimitDegrees=30, swing2LimitDegrees=30, twistLimitDegrees=10}
 {bone1Name=dyn_r_tentacle_a_04, bone2Name=dyn_r_tentacle_a_03, swing1LimitDegrees=30, swing2LimitDegrees=30, twistLimitDegrees=10}
 {bone1Name=dyn_r_tentacle_a_05, bone2Name=dyn_r_tentacle_a_04, swing1LimitDegrees=30, swing2LimitDegrees=30, twistLimitDegrees=10}
 {
  bone1Name=dyn_r_tentacle_b_01
  bone2Name=pack
  swing1LimitDegrees=45
  swing1Motion=Free
  swing2LimitDegrees=45
  swing2Motion=Free
  twistLimitDegrees=45
  twistMotion=Free
 }
 {bone1Name=dyn_r_tentacle_b_02, bone2Name=dyn_r_tentacle_b_01, swing1LimitDegrees=30, swing2LimitDegrees=30, twistLimitDegrees=10}
 {bone1Name=dyn_r_tentacle_b_03, bone2Name=dyn_r_tentacle_b_02, swing1LimitDegrees=30, swing2LimitDegrees=30, twistLimitDegrees=10}
 {bone1Name=dyn_r_tentacle_b_04, bone2Name=dyn_r_tentacle_b_03, swing1LimitDegrees=30, swing2LimitDegrees=30, twistLimitDegrees=10}
 {bone1Name=dyn_r_tentacle_b_05, bone2Name=dyn_r_tentacle_b_04, swing1LimitDegrees=30, swing2LimitDegrees=30, twistLimitDegrees=10}
 {
  bone1Name=dyn_r_tentacle_c_01
  bone2Name=pack
  swing1LimitDegrees=45
  swing1Motion=Free
  swing2LimitDegrees=45
  swing2Motion=Free
  twistLimitDegrees=45
  twistMotion=Free
 }
 {bone1Name=dyn_r_tentacle_c_02, bone2Name=dyn_r_tentacle_c_01, swing1LimitDegrees=30, swing2LimitDegrees=30, twistLimitDegrees=10}
 {bone1Name=dyn_r_tentacle_c_03, bone2Name=dyn_r_tentacle_c_02, swing1LimitDegrees=30, swing2LimitDegrees=30, twistLimitDegrees=10}
 {bone1Name=dyn_r_tentacle_c_04, bone2Name=dyn_r_tentacle_c_03, swing1LimitDegrees=30, swing2LimitDegrees=30, twistLimitDegrees=10}
 {bone1Name=dyn_r_tentacle_c_05, bone2Name=dyn_r_tentacle_c_04, swing1LimitDegrees=30, swing2LimitDegrees=30, twistLimitDegrees=10}
]
